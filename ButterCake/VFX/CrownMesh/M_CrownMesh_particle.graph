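# Default Lit material for the Victory Crown mesh, enabled for Niagara mesh particles, skeletal meshes and mesh deformers.
# It samples the crown item's diffuse, FX mask, normal and S mask textures, alongside a scalar constant.
# The diffuse color is multiplied by the FX mask's green channel and fed into MF_FX_EmissiveAdjustments as emissive, confining glow to masked areas.
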Material /ButterCake/VFX/CrownMesh/M_CrownMesh_particle
graph {
 properties {
  bUsedWithMeshDeformer=true
  bUsedWithNiagaraMeshParticles=true
  bUsedWithSkeletalMesh=true
  pixelDepthOffsetMode=PDOM_Legacy
  usedShadingModels=Default Lit
 }
 nodes [
  Constant() {r=0.3}
  n1 = TextureCoordinate()
  MaterialFunctionCall(Emissive=Multiply(
    A=TextureSample(UVs=@n1) {texture=/VictoryCrownsGameplay/Items/Assets/Textures/T_Item_Victory_Crown_D}.RGB
    B=TextureSample() {samplerType=SAMPLERTYPE_Masks, texture=/VictoryCrownsGameplay/Items/Assets/Textures/T_Item_Victory_Crown_FX}.G
   )) {materialFunction=&MF_FX_EmissiveAdjustments}
  TextureSample() {samplerType=SAMPLERTYPE_Normal, texture=/VictoryCrownsGameplay/Items/Assets/Textures/T_Item_Victory_Crown_N}
  TextureSample() {samplerType=SAMPLERTYPE_Masks, texture=/VictoryCrownsGameplay/Items/Assets/Textures/T_Item_Victory_Crown_S}
 ]
}
